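; Character part modifier Blueprint, apparently a test, for applying wraps to a player's weapons as overlay materials.
; SetOverlayWrap prints debug text and passes the current weapon mesh and an indexed material to SetMaterialParams.
; That function creates a dynamic overlay, copying bounds vectors and the customization mask from the weapon's material.
; A separate, uncalled-here helper server-applies an override wrap to items via the associated player controller; weapon and pickup events only call parents.
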(event EventOnNewWeapons
  (params (WeaponsAsItems :type Array%20of%20Fort%20Item%20Object%20References))
  (|CallParentFunction WeaponsAsItems))

(event Custom|SetOverlayWrap
  (params (MaterialIndex :type Integer))
  (bind _returnvalue (CustomCharacter|GetAssociatedPlayerPawn))
  (bind _currentweapon (Variables|/Script/FortniteGame.FortPawn|GetCurrentWeapon _returnvalue))
  (bind _weaponmesh (Variables|/Script/FortniteGame.FortWeapon|GetWeaponMesh _currentweapon))
  (Development|PrintString (Utilities|String|Append (Utilities|GetDisplayName _weaponmesh) "  ||  " (Utilities|String|ToString(Integer) MaterialIndex)) true true "(R=0.000000,G=0.660000,B=1.000000,A=1.000000)" 1e+01)
  (Default|SetMaterialParams _weaponmesh (Utilities|Array|Get(acopy) (Variables|Default|GetMaterials) MaterialIndex)))

(event EventOnPlayerPickupDropped
  (params (Pickup :type Fort%20Pickup%20Object%20Reference) (PickupItemEntry :type Fort%20Item%20Entry%20Structure%20%28by%20ref%29))
  (|CallParentFunction Pickup PickupItemEntry))

(fn ApplyOverrideToItems
  (params (Array :type Array%20of%20Fort%20Item%20Object%20References))
  (bind _asfort_player_controller (Default|GetAssociatedPlayerController))
  (Utilities|IsValid _asfort_player_controller
    (:Is%20Valid
      (for _array_element Array
        (Class|FortPlayerController|ServerApplyOverrideWraptoItem _asfort_player_controller (Item|GetItemGuid _array_element) (Variables|Default|GetWeapon%20Wrap%20to%20Use))))))

(fn GetAssociatedPlayerController
  (bind _returnvalue (CustomCharacter|GetAssociatedPlayerPawn))
  (Utilities|FlowControl|Sequence
    (:then_0
      (bind _aspawn (Utilities|Casting|CastToPawn (Vehicle|GetVehicleActor _returnvalue :!class /Script/FortniteGame.FortPlayerPawn)))
      (bind _asfort_player_controller (Utilities|Casting|CastToFortPlayerController (Pawn|GetController _aspawn)))
      (return _asfort_player_controller))
    (:then_1
      (bind _asfort_player_controller_2 (Utilities|Casting|CastToFortPlayerController (Pawn|GetController _returnvalue)))
      (return _asfort_player_controller_2))))

(fn OnRep_WeaponIndex)

(fn SetMaterialParams
  (params (SkelMesh :type Skeletal%20Mesh%20Component%20Object%20Reference) (OverlayMaterial :type Material%20Interface%20Object%20Reference))
  (bind _material_instance_dynamic (Class|BTechArtGlobalLibrary|GetDynamicMaterialInstance SkelMesh))
  (bind _returnvalue (Rendering|Material|CreateDynamicMaterialInstance OverlayMaterial "None" "None" :!class /Script/Engine.KismetMaterialLibrary))
  (bind _returnvalue_2 (Rendering|Material|GetVectorParameterValue _material_instance_dynamic "DO NOT OVERRIDE-OBCENTER" :!class /Script/Engine.MaterialInstanceDynamic))
  (bind _returnvalue_3 (Rendering|Material|GetVectorParameterValue _material_instance_dynamic "DO NOT OVERRIDE-OBMIN" :!class /Script/Engine.MaterialInstanceDynamic))
  (bind _returnvalue_4 (Rendering|Material|GetVectorParameterValue _material_instance_dynamic "DO NOT OVERRIDE-OBMAX" :!class /Script/Engine.MaterialInstanceDynamic))
  (bind _returnvalue_5 (Rendering|Material|GetTextureParameterValue _material_instance_dynamic "CustomizationMask (_CM)" :!class /Script/Engine.MaterialInstanceDynamic))
  (Rendering|Material|SetVectorParameterValue _returnvalue "DO NOT OVERRIDE-OBCENTER" _returnvalue_2 :!class /Script/Engine.MaterialInstanceDynamic)
  (Rendering|Material|SetVectorParameterValue _returnvalue "DO NOT OVERRIDE-OBMIN" _returnvalue_3 :!class /Script/Engine.MaterialInstanceDynamic)
  (Rendering|Material|SetVectorParameterValue _returnvalue "DO NOT OVERRIDE-OBMAX" _returnvalue_4 :!class /Script/Engine.MaterialInstanceDynamic)
  (Rendering|Material|SetTextureParameterValue _returnvalue "CustomizationMask (_CM)" _returnvalue_5)
  (Rendering|Material|SetOverlayMaterial SkelMesh _returnvalue false 0 :!class /Script/Engine.MeshComponent))

(fn ConstructionScript
  (|CallParentFunction))
